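; Camera modifier that animates view roll and FOV for weapon fire, using curves from a WeaponFireCameraAnimationData class.
; BlueprintModifyCamera skips its logic when UseNativeImplementation is true; otherwise it samples curves by time since StartTime.
; A continuous-fire timer, reset once TimeBetweenShots is exceeded, drives an intensity scale for curve roll, added Perlin roll noise and additive FOV.
; Getter and setter functions expose timing state and shake data.
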(fn BlueprintModifyCamera
  (params (DeltaTime :type Float%20%28single-precision%29) (ViewLocation :type Vector :default "0, 0, 0") (ViewRotation :type Rotator :default "0, 0, 0") (FOV :type Float%20%28single-precision%29))
  (bind _elapsedtime (LocalVariables|GetElapsedTime))
  (bind _shakedata (Variables|Default|GetShakeData))
  (bind (_rollcurve _fovcurve) (Class|GetClassDefaults _shakedata))
  (bind _returnvalue (Math|Curves|GetFloatValue _fovcurve _elapsedtime))
  (bind _shakedata_2 (Variables|Default|GetShakeData))
  (bind (_rollcurve_2 _fovcurve_2 _rollnoisecurve _continuousfireintensitycurve _rollnoisetimescale) (Class|GetClassDefaults _shakedata_2))
  (bind _shakedata_3 (Variables|Default|GetShakeData))
  (bind (_rollcurve_3 _fovcurve_3 _rollnoisecurve_2 _continuousfireintensitycurve_2) (Class|GetClassDefaults _shakedata_3))
  (bind _shakedata_4 (Variables|Default|GetShakeData))
  (bind _rollcurve_4 (Class|GetClassDefaults _shakedata_4))
  (bind _intensityscale (LocalVariables|GetIntensityScale))
  (bind _returnvalue_2 (* _returnvalue _intensityscale))
  (bind _viewrotation (Variables|GetViewRotation))
  (if (CallFunction|UseNativeImplementation)
    _
    (elif (Utilities|IsValidClass (Variables|Default|GetShakeData))
      (Variables|Default|SetIntensityTimer (+ (Variables|GetDeltaTime) (Variables|Default|GetIntensityTimer)))
      (Utilities|FlowControl|Sequence
        (:then_0
          (LocalVariables|SetFinalPitch (.pitch _viewrotation)))
        (:then_1
          (LocalVariables|SetFinalRoll (.roll _viewrotation)))
        (:then_2
          (LocalVariables|SetFinalYaw (.yaw _viewrotation)))
        (:then_3
          (LocalVariables|SetFinalFOV (Variables|GetFOV)))
        (:then_4
          (LocalVariables|SetFinalLocation (Variables|GetViewLocation)))
        (:then_5
          (LocalVariables|SetElapsedTime (- (Utilities|Time|GetTimeSeconds) (Variables|Default|GetStartTime)))
          ;[1 1 1] Animated Roll
          (if (> (LocalVariables|GetElapsedTime) (Variables|Default|GetTimeBetweenShots))
            (Variables|Default|SetIntensityTimer 0.0))
          (Utilities|IsValid _continuousfireintensitycurve_2
            (:Is%20Valid
              (LocalVariables|SetIntensityScale (Math|Curves|GetFloatValue _continuousfireintensitycurve_2 (Variables|Default|GetIntensityTimer))))
            (:Is%20Not%20Valid
              (LocalVariables|SetIntensityScale 1.0)))
          (Utilities|IsValid _rollcurve_4
            (:Is%20Valid
              (LocalVariables|SetFinalRoll (* (Math|Curves|GetFloatValue _rollcurve_4 (LocalVariables|GetElapsedTime)) (LocalVariables|GetIntensityScale))))
            (:Is%20Not%20Valid
              _))
          ;[1 0.7516 0.2317] Roll Noise
          (Utilities|IsValid _rollnoisecurve
            (:Is%20Valid
              (LocalVariables|SetFinalRoll (+ (LocalVariables|GetFinalRoll) (* (Math|Random|PerlinNoise1D (LocalVariables|GetElapsedTime)) (* (Math|Curves|GetFloatValue _rollnoisecurve (* (LocalVariables|GetElapsedTime) _rollnoisetimescale)) (LocalVariables|GetIntensityScale))))))
            (:Is%20Not%20Valid
              _))
          (LocalVariables|SetFinalRotation (Math|Rotator|MakeRotator (LocalVariables|GetFinalRoll) (LocalVariables|GetFinalPitch) (LocalVariables|GetFinalYaw)))
          ;[1 1 1] Additive FOV
          (Utilities|IsValid _fovcurve
            (:Is%20Valid
              (LocalVariables|SetFinalFOV (- (LocalVariables|GetFinalFOV) (Variables|Default|GetPreviousDeltaFOV)))
              (LocalVariables|SetFinalFOV (+ _returnvalue_2 (Variables|GetFOV)))
              (Variables|Default|SetPreviousDeltaFOV _returnvalue_2))
            (:Is%20Not%20Valid
              _))
          (return (LocalVariables|GetFinalLocation) (LocalVariables|GetFinalRotation) (LocalVariables|GetFinalFOV)))))))

(fn GetCameraShakeData
  (bind (_rollcurve _fovcurve _rollnoisecurve _continuousfireintensitycurve _rollnoisetimescale) (Class|GetClassDefaults (Variables|Default|GetShakeData)))
  (if (Utilities|IsValidClass (Variables|Default|GetShakeData))
    (return _rollcurve _fovcurve _rollnoisecurve _continuousfireintensitycurve _rollnoisetimescale true)
    (else
      (return 0 0 0 0 0.0 false))))

(fn GetInputParams
  (return (Variables|Default|GetStartTime) (Variables|Default|GetPreviousDeltaFOV) (Variables|Default|GetIntensityTimer) (Variables|Default|GetTimeBetweenShots)))

(fn SetInputParams
  (params (StartTime :type Float%20%28single-precision%29) (PreviousDeltaFOV :type Float%20%28single-precision%29) (IntensityTimer :type Float%20%28single-precision%29) (TimeBetweenShots :type Float%20%28single-precision%29))
  (Variables|Default|SetStartTime StartTime)
  (Variables|Default|SetPreviousDeltaFOV PreviousDeltaFOV)
  (Variables|Default|SetIntensityTimer IntensityTimer)
  (Variables|Default|SetTimeBetweenShots TimeBetweenShots))

(fn SetShakeDataParam
  (params (ShakeDataClass :type Primary%20Data%20Asset%20Class%20Reference))
  (bind _asweapon_fire_camera_animation_data (Utilities|Casting|CastToWeaponFireCameraAnimationDataClass ShakeDataClass))
  (Variables|Default|SetShakeData _asweapon_fire_camera_animation_data))
